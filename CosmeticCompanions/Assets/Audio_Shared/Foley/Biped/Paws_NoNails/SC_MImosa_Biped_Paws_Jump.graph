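# Sound cue for the paw jump foley of a biped cosmetic companion, from the nail-less paws set.
# A quality-level switch feeds the first two slots from a random pick of seven jump waves and the last two from a smaller pool of the first four.
# A root modulator slightly varies pitch and volume per play.
# It uses companion foley attenuation, the pets sound class and shared biped footstep concurrency.
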SoundCue /CosmeticCompanions/Assets/Audio_Shared/Foley/Biped/Paws_NoNails/SC_MImosa_Biped_Paws_Jump
root=Modulator_0
nodes [
 {id=Modulator_0, children [QualityLevel_0], properties {pitchMax=1.1, pitchMin=0.9, volumeMax=1, volumeMin=0.9}}
 {id=QualityLevel_0, children [Random_2,Random_2,Random_3,Random_3]}
 {
  id=Random_2
  children [WavePlayer_23, WavePlayer_24, WavePlayer_25, WavePlayer_26, WavePlayer_27, WavePlayer_28, WavePlayer_29]
  properties {weights [1,1,1,1,1,1,1]}
 }
 {id=Random_3, children [WavePlayer_30, WavePlayer_31, WavePlayer_32, WavePlayer_33], properties {weights [1,1,1,1]}}
 {
  id=WavePlayer_23
  properties {soundWaveAssetPtr=/CosmeticCompanions/Assets/Audio_Shared/Foley/Biped/Paws_NoNails/Wavs/Mimosa_Biped_Paws_Jump_01}
 }
 {
  id=WavePlayer_24
  properties {soundWaveAssetPtr=/CosmeticCompanions/Assets/Audio_Shared/Foley/Biped/Paws_NoNails/Wavs/Mimosa_Biped_Paws_Jump_02}
 }
 {
  id=WavePlayer_25
  properties {soundWaveAssetPtr=/CosmeticCompanions/Assets/Audio_Shared/Foley/Biped/Paws_NoNails/Wavs/Mimosa_Biped_Paws_Jump_03}
 }
 {
  id=WavePlayer_26
  properties {soundWaveAssetPtr=/CosmeticCompanions/Assets/Audio_Shared/Foley/Biped/Paws_NoNails/Wavs/Mimosa_Biped_Paws_Jump_04}
 }
 {
  id=WavePlayer_27
  properties {soundWaveAssetPtr=/CosmeticCompanions/Assets/Audio_Shared/Foley/Biped/Paws_NoNails/Wavs/Mimosa_Biped_Paws_Jump_05}
 }
 {
  id=WavePlayer_28
  properties {soundWaveAssetPtr=/CosmeticCompanions/Assets/Audio_Shared/Foley/Biped/Paws_NoNails/Wavs/Mimosa_Biped_Paws_Jump_06}
 }
 {
  id=WavePlayer_29
  properties {soundWaveAssetPtr=/CosmeticCompanions/Assets/Audio_Shared/Foley/Biped/Paws_NoNails/Wavs/Mimosa_Biped_Paws_Jump_07}
 }
 {
  id=WavePlayer_30
  properties {soundWaveAssetPtr=&Mimosa_Biped_Paws_Jump_01}
 }
 {
  id=WavePlayer_31
  properties {soundWaveAssetPtr=&Mimosa_Biped_Paws_Jump_02}
 }
 {
  id=WavePlayer_32
  properties {soundWaveAssetPtr=&Mimosa_Biped_Paws_Jump_03}
 }
 {
  id=WavePlayer_33
  properties {soundWaveAssetPtr=&Mimosa_Biped_Paws_Jump_04}
 }
]
properties {
 attenuationSettings=/Game/Sounds/Attenuation/AS_Companion_Foley
 maxDistance=1200
 priority=0
 soundClassObject=&Fort_Creatures_Pets
 volumeMultiplier=0.85
 concurrencySet [/CosmeticCompanions/Assets/Audio_Shared/Foley/Biped/CONC_Mimosa_Biped_Footstep]
}
