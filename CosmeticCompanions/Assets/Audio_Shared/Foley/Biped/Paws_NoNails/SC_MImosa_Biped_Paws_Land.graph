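# Sound cue for the landing foley of a biped cosmetic companion (Mimosa) with nail-less paws.
# A quality-level switch feeds either a seven-variation random pool or, for its last two levels, a smaller pool of the first four waves.
# A top-level modulator varies pitch and volume on each play.
# It routes to the pets sound class, using companion foley attenuation and the shared biped footstep concurrency.
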SoundCue /CosmeticCompanions/Assets/Audio_Shared/Foley/Biped/Paws_NoNails/SC_MImosa_Biped_Paws_Land
root=Modulator_0
nodes [
 {id=Modulator_0, children [QualityLevel_0], properties {pitchMax=1.1, pitchMin=0.9, volumeMax=1, volumeMin=0.8}}
 {id=QualityLevel_0, children [Random_0,Random_0,Random_3,Random_3]}
 {
  id=Random_0
  children [WavePlayer_0,WavePlayer_1,WavePlayer_2,WavePlayer_3,WavePlayer_4,WavePlayer_5,WavePlayer_6]
  properties {weights [1,1,1,1,1,1,1]}
 }
 {id=Random_3, children [WavePlayer_7, WavePlayer_8, WavePlayer_9, WavePlayer_10], properties {weights [1,1,1,1]}}
 {
  id=WavePlayer_0
  properties {soundWaveAssetPtr=/CosmeticCompanions/Assets/Audio_Shared/Foley/Biped/Paws_NoNails/Wavs/Mimosa_Biped_Paws_Land_01}
 }
 {
  id=WavePlayer_1
  properties {soundWaveAssetPtr=/CosmeticCompanions/Assets/Audio_Shared/Foley/Biped/Paws_NoNails/Wavs/Mimosa_Biped_Paws_Land_02}
 }
 {
  id=WavePlayer_10
  properties {soundWaveAssetPtr=/CosmeticCompanions/Assets/Audio_Shared/Foley/Biped/Paws_NoNails/Wavs/Mimosa_Biped_Paws_Land_04}
 }
 {
  id=WavePlayer_2
  properties {soundWaveAssetPtr=/CosmeticCompanions/Assets/Audio_Shared/Foley/Biped/Paws_NoNails/Wavs/Mimosa_Biped_Paws_Land_03}
 }
 {
  id=WavePlayer_3
  properties {soundWaveAssetPtr=&Mimosa_Biped_Paws_Land_04}
 }
 {
  id=WavePlayer_4
  properties {soundWaveAssetPtr=/CosmeticCompanions/Assets/Audio_Shared/Foley/Biped/Paws_NoNails/Wavs/Mimosa_Biped_Paws_Land_05}
 }
 {
  id=WavePlayer_5
  properties {soundWaveAssetPtr=/CosmeticCompanions/Assets/Audio_Shared/Foley/Biped/Paws_NoNails/Wavs/Mimosa_Biped_Paws_Land_06}
 }
 {
  id=WavePlayer_6
  properties {soundWaveAssetPtr=/CosmeticCompanions/Assets/Audio_Shared/Foley/Biped/Paws_NoNails/Wavs/Mimosa_Biped_Paws_Land_07}
 }
 {
  id=WavePlayer_7
  properties {soundWaveAssetPtr=&Mimosa_Biped_Paws_Land_01}
 }
 {
  id=WavePlayer_8
  properties {soundWaveAssetPtr=&Mimosa_Biped_Paws_Land_02}
 }
 {
  id=WavePlayer_9
  properties {soundWaveAssetPtr=&Mimosa_Biped_Paws_Land_03}
 }
]
properties {
 attenuationSettings=/Game/Sounds/Attenuation/AS_Companion_Foley
 maxDistance=1200
 pitchMultiplier=0.9
 priority=0
 soundClassObject=&Fort_Creatures_Pets
 volumeMultiplier=1
 concurrencySet [/CosmeticCompanions/Assets/Audio_Shared/Foley/Biped/CONC_Mimosa_Biped_Footstep]
}
